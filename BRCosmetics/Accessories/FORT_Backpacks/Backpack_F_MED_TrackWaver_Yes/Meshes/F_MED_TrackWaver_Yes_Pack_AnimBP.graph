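# Anim Blueprint for the TrackWaver_Yes winged backpack, posing its wings to match player movement.
# A movement state machine, with nested skydiving states, blends named poses and directional lean blendspaces, then adds an idle/idle-break layer.
# AnimDynamics springs jiggle both wings and their feathers, before bone offsets apply back spacing and fold-weighted rotation.
# An older MistRaven-based state machine and idle player remain in the graph, disconnected from the output.
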AnimBlueprint /BRCosmetics/Accessories/FORT_Backpacks/Backpack_F_MED_TrackWaver_Yes/Meshes/F_MED_TrackWaver_Yes_Pack_AnimBP
target_skeleton=&/F_MED_TrackWaver_Yes_Pack_Skeleton
graphs [
 AnimGraph {nodes [
   Input Pose()
   AnimGraphNode_StateMachine_0 = IdleBreak() {is=title AnimGraphNode_StateMachine}
   AnimGraphNode_StateMachine_1 = wingPosition_1() {is=title AnimGraphNode_StateMachine}
   n1 = Anim Dynamics - Bone: wing_L(
    ComponentPose=Local To Component(LocalPose=Apply Additive(Base=@AnimGraphNode_StateMachine_1, Additive=@AnimGraphNode_StateMachine_0) {alphaBoolBlend {blendOption=Linear}})
    Alpha=1
    ExternalForce="0,0,0"
   ) {
    angularDampingOverride=0.99
    angularSpringConstant=300
    bAngularSpring=true
    bGravityOverrideInSimSpace=true
    bOverrideAngularDamping=true
    gravityScale=0.05
    numSolverIterationsPostUpdate=4
    boundBone {boneName=wing_L}
    componentAppliedLinearAccClamp=<1000 1000 1000>
    componentLinearAccScale=<0.5 0.5 0.5>
    componentLinearVelScale=<0.5 0.5 0.5>
    physicsBodyDefinitions [{
      boundBone {boneName=wing_L}
      boxExtents=<5 5 5>
      constraintSetup {angularLimitsMax=<20 20 40>, angularLimitsMin=<-20 -20 -40>, angularTarget=<0.1 0 0>}
      localJointOffset=<-0.1 0 0>
     }]
    relativeSpaceBone {boneName=root}
    simSpaceSettings {maxAngularAcceleration=5, maxAngularVelocity=25, simSpaceAngularAlpha=1}
   }
   n2 = Anim Dynamics - Bone: featherE_L(ComponentPose=@n1, Alpha=1, ExternalForce="0,0,0") {
    angularDampingOverride=0.99
    angularSpringConstant=150
    bAngularSpring=true
    bGravityOverrideInSimSpace=true
    bOverrideAngularDamping=true
    gravityScale=0.1
    numSolverIterationsPostUpdate=4
    boundBone {boneName=featherE_L}
    componentAppliedLinearAccClamp=<300 300 300>
    componentLinearAccScale=<0.2 0.2 0.2>
    componentLinearVelScale=<0.2 0.2 0.2>
    physicsBodyDefinitions [{
      boundBone {boneName=featherE_L}
      boxExtents=<5 5 5>
      constraintSetup {angularLimitsMax=<0 0 40>, angularLimitsMin=<0 0 -40>, angularTarget=<0.1 0 0>}
      localJointOffset=<-0.05 0 0>
     }]
    relativeSpaceBone {boneName=root}
    simSpaceSettings {maxAngularAcceleration=2, maxAngularVelocity=5, simSpaceAngularAlpha=1}
   }
   n3 = Anim Dynamics - Bone: featherD_L(ComponentPose=@n2, Alpha=1, ExternalForce="0,0,0") {
    angularDampingOverride=0.99
    angularSpringConstant=150
    bAngularSpring=true
    bGravityOverrideInSimSpace=true
    bOverrideAngularDamping=true
    gravityScale=0.1
    numSolverIterationsPostUpdate=4
    boundBone {boneName=featherD_L}
    componentAppliedLinearAccClamp=<300 300 300>
    componentLinearAccScale=<0.2 0.2 0.2>
    componentLinearVelScale=<0.2 0.2 0.2>
    physicsBodyDefinitions [{
      boundBone {boneName=featherD_L}
      boxExtents=<4.8 4.8 4.8>
      constraintSetup {angularLimitsMax=<0 0 40>, angularLimitsMin=<0 0 -40>, angularTarget=<0.1 0 0>}
      localJointOffset=<-0.05 0 0>
     }]
    relativeSpaceBone {boneName=root}
    simSpaceSettings {maxAngularAcceleration=2, maxAngularVelocity=5, simSpaceAngularAlpha=1}
   }
   n4 = Anim Dynamics - Bone: featherC_L(ComponentPose=@n3, Alpha=1, ExternalForce="0,0,0") {
    angularDampingOverride=0.99
    angularSpringConstant=150
    bAngularSpring=true
    bGravityOverrideInSimSpace=true
    bOverrideAngularDamping=true
    gravityScale=0.1
    numSolverIterationsPostUpdate=4
    boundBone {boneName=featherC_L}
    componentAppliedLinearAccClamp=<300 300 300>
    componentLinearAccScale=<0.2 0.2 0.2>
    componentLinearVelScale=<0.2 0.2 0.2>
    physicsBodyDefinitions [{
      boundBone {boneName=featherC_L}
      boxExtents=<4.6 4.6 4.6>
      constraintSetup {angularLimitsMax=<0 0 40>, angularLimitsMin=<0 0 -40>, angularTarget=<0.1 0 0>}
      localJointOffset=<-0.05 0 0>
     }]
    relativeSpaceBone {boneName=root}
    simSpaceSettings {maxAngularAcceleration=2, maxAngularVelocity=5, simSpaceAngularAlpha=1}
   }
   n5 = Anim Dynamics - Bone: featherB_L(ComponentPose=@n4, Alpha=1, ExternalForce="0,0,0") {
    angularDampingOverride=0.99
    angularSpringConstant=150
    bAngularSpring=true
    bGravityOverrideInSimSpace=true
    bOverrideAngularDamping=true
    gravityScale=0.1
    numSolverIterationsPostUpdate=4
    boundBone {boneName=featherB_L}
    componentAppliedLinearAccClamp=<300 300 300>
    componentLinearAccScale=<0.2 0.2 0.2>
    componentLinearVelScale=<0.2 0.2 0.2>
    physicsBodyDefinitions [{
      boundBone {boneName=featherB_L}
      boxExtents=<4.6 4.6 4.6>
      constraintSetup {angularLimitsMax=<0 0 40>, angularLimitsMin=<0 0 -40>, angularTarget=<0.1 0 0>}
      localJointOffset=<-0.05 0 0>
     }]
    relativeSpaceBone {boneName=root}
    simSpaceSettings {maxAngularAcceleration=2, maxAngularVelocity=5, simSpaceAngularAlpha=1}
   }
   n6 = Anim Dynamics - Bone: featherA_L(ComponentPose=@n5, Alpha=1, ExternalForce="0,0,0") {
    angularDampingOverride=0.99
    angularSpringConstant=150
    bAngularSpring=true
    bGravityOverrideInSimSpace=true
    bOverrideAngularDamping=true
    gravityScale=0.1
    numSolverIterationsPostUpdate=4
    boundBone {boneName=featherA_L}
    componentAppliedLinearAccClamp=<300 300 300>
    componentLinearAccScale=<0.2 0.2 0.2>
    componentLinearVelScale=<0.2 0.2 0.2>
    physicsBodyDefinitions [{
      boundBone {boneName=featherA_L}
      boxExtents=<4.6 4.6 4.6>
      constraintSetup {angularLimitsMax=<0 0 40>, angularLimitsMin=<0 0 -40>, angularTarget=<0.1 0 0>}
      localJointOffset=<-0.05 0 0>
     }]
    relativeSpaceBone {boneName=root}
    simSpaceSettings {maxAngularAcceleration=2, maxAngularVelocity=5, simSpaceAngularAlpha=1}
   }
   n7 = Anim Dynamics - Bone: wing_R(ComponentPose=@n6, Alpha=1, ExternalForce="0,0,0") {
    angularDampingOverride=0.99
    angularSpringConstant=300
    bAngularSpring=true
    bGravityOverrideInSimSpace=true
    bOverrideAngularDamping=true
    gravityScale=0.05
    numSolverIterationsPostUpdate=4
    boundBone {boneName=wing_R}
    componentAppliedLinearAccClamp=<1000 1000 1000>
    componentLinearAccScale=<0.5 0.5 0.5>
    componentLinearVelScale=<0.5 0.5 0.5>
    physicsBodyDefinitions [{
      boundBone {boneName=wing_R}
      boxExtents=<5 5 5>
      constraintSetup {angularLimitsMax=<20 20 40>, angularLimitsMin=<-20 -20 -40>, angularTarget=<0.1 0 0>}
      localJointOffset=<0.1 0 0>
     }]
    relativeSpaceBone {boneName=root}
    simSpaceSettings {maxAngularAcceleration=5, maxAngularVelocity=25, simSpaceAngularAlpha=1}
   }
   n8 = Anim Dynamics - Bone: featherE_R(ComponentPose=@n7, Alpha=1, ExternalForce="0,0,0") {
    angularDampingOverride=0.99
    angularSpringConstant=150
    bAngularSpring=true
    bGravityOverrideInSimSpace=true
    bOverrideAngularDamping=true
    gravityScale=0.1
    numSolverIterationsPostUpdate=4
    boundBone {boneName=featherE_R}
    componentAppliedLinearAccClamp=<300 300 300>
    componentLinearAccScale=<0.2 0.2 0.2>
    componentLinearVelScale=<0.2 0.2 0.2>
    physicsBodyDefinitions [{
      boundBone {boneName=featherE_R}
      boxExtents=<5 5 5>
      constraintSetup {angularLimitsMax=<0 0 40>, angularLimitsMin=<0 0 -40>, angularTarget=<0.1 0 0>}
      localJointOffset=<0.05 0 0>
     }]
    relativeSpaceBone {boneName=root}
    simSpaceSettings {maxAngularAcceleration=2, maxAngularVelocity=5, simSpaceAngularAlpha=1}
   }
   n9 = Anim Dynamics - Bone: featherD_R(ComponentPose=@n8, Alpha=1, ExternalForce="0,0,0") {
    angularDampingOverride=0.99
    angularSpringConstant=150
    bAngularSpring=true
    bGravityOverrideInSimSpace=true
    bOverrideAngularDamping=true
    gravityScale=0.1
    numSolverIterationsPostUpdate=4
    boundBone {boneName=featherD_R}
    componentAppliedLinearAccClamp=<300 300 300>
    componentLinearAccScale=<0.2 0.2 0.2>
    componentLinearVelScale=<0.2 0.2 0.2>
    physicsBodyDefinitions [{
      boundBone {boneName=featherD_R}
      boxExtents=<4.8 4.8 4.8>
      constraintSetup {angularLimitsMax=<0 0 40>, angularLimitsMin=<0 0 -40>, angularTarget=<0.1 0 0>}
      localJointOffset=<0.05 0 0>
     }]
    relativeSpaceBone {boneName=root}
    simSpaceSettings {maxAngularAcceleration=2, maxAngularVelocity=5, simSpaceAngularAlpha=1}
   }
   n10 = Anim Dynamics - Bone: featherC_R(ComponentPose=@n9, Alpha=1, ExternalForce="0,0,0") {
    angularDampingOverride=0.99
    angularSpringConstant=150
    bAngularSpring=true
    bGravityOverrideInSimSpace=true
    bOverrideAngularDamping=true
    gravityScale=0.1
    numSolverIterationsPostUpdate=4
    boundBone {boneName=featherC_R}
    componentAppliedLinearAccClamp=<300 300 300>
    componentLinearAccScale=<0.2 0.2 0.2>
    componentLinearVelScale=<0.2 0.2 0.2>
    physicsBodyDefinitions [{
      boundBone {boneName=featherC_R}
      boxExtents=<4.6 4.6 4.6>
      constraintSetup {angularLimitsMax=<0 0 40>, angularLimitsMin=<0 0 -40>, angularTarget=<0.1 0 0>}
      localJointOffset=<0.05 0 0>
     }]
    relativeSpaceBone {boneName=root}
    simSpaceSettings {maxAngularAcceleration=2, maxAngularVelocity=5, simSpaceAngularAlpha=1}
   }
   n11 = Anim Dynamics - Bone: featherB_R(ComponentPose=@n10, Alpha=1, ExternalForce="0,0,0") {
    angularDampingOverride=0.99
    angularSpringConstant=150
    bAngularSpring=true
    bGravityOverrideInSimSpace=true
    bOverrideAngularDamping=true
    gravityScale=0.1
    numSolverIterationsPostUpdate=4
    boundBone {boneName=featherB_R}
    componentAppliedLinearAccClamp=<300 300 300>
    componentLinearAccScale=<0.2 0.2 0.2>
    componentLinearVelScale=<0.2 0.2 0.2>
    physicsBodyDefinitions [{
      boundBone {boneName=featherB_R}
      boxExtents=<4.6 4.6 4.6>
      constraintSetup {angularLimitsMax=<0 0 40>, angularLimitsMin=<0 0 -40>, angularTarget=<0.1 0 0>}
      localJointOffset=<0.05 0 0>
     }]
    relativeSpaceBone {boneName=root}
    simSpaceSettings {maxAngularAcceleration=2, maxAngularVelocity=5, simSpaceAngularAlpha=1}
   }
   n12 = Anim Dynamics - Bone: featherA_R(ComponentPose=@n11, Alpha=1, ExternalForce="0,0,0") {
    angularDampingOverride=0.99
    angularSpringConstant=150
    bAngularSpring=true
    bGravityOverrideInSimSpace=true
    bOverrideAngularDamping=true
    gravityScale=0.1
    numSolverIterationsPostUpdate=4
    boundBone {boneName=featherA_R}
    componentAppliedLinearAccClamp=<300 300 300>
    componentLinearAccScale=<0.2 0.2 0.2>
    componentLinearVelScale=<0.2 0.2 0.2>
    physicsBodyDefinitions [{
      boundBone {boneName=featherA_R}
      boxExtents=<4.6 4.6 4.6>
      constraintSetup {angularLimitsMax=<0 0 40>, angularLimitsMin=<0 0 -40>, angularTarget=<0.1 0 0>}
      localJointOffset=<0.05 0 0>
     }]
    relativeSpaceBone {boneName=root}
    simSpaceSettings {maxAngularAcceleration=2, maxAngularVelocity=5, simSpaceAngularAlpha=1}
   }
   n13 = Local To Component(LocalPose=Component To Local(ComponentPose=@n12))
   n14 = Get BackSpacing() -> (BackSpacing:double)
   n15 = "Transform (Modify) Bone - Bone: wing_L"(
    ComponentPose="Transform (Modify) Bone - Bone: wing_R"!(
     ComponentPose="Transform (Modify) Bone - Bone: wing_L"!(ComponentPose=@n13, Alpha=1, Translation_X:double=1.3, Translation_Y:double=@n14, Translation_Z:double=0) -> (Pose) {translationMode=BMM_Additive, boneToModify {boneName=wing_L}, translation=<1.3 1>}
     Alpha=1
     Translation_X:double=-1
     Translation_Y:double=@n14
     Translation_Z:double=0
    ) -> (Pose) {translationMode=BMM_Additive, boneToModify {boneName=wing_R}, translation=<-1 1>}
    Alpha=Get wings fold() -> (wings fold:float)
    Translation="0,0,0"
    Rotation="-10,-15,-15"
    Scale="1,1,1"
   ) {rotationMode=BMM_Additive, boneToModify {boneName=wing_L}, rotation=<rot -10 -15 -15>}
   n16 = Get wings fold()
   Output Pose(Result=Component To Local(ComponentPose="Transform (Modify) Bone - Bone: wing_R"!(ComponentPose=@n15, Alpha=@n16, Translation:Vector="0,0,0", Rotation:Rotator="10,15,-15", Scale:Vector="1,1,1") -> (Pose) {rotationMode=BMM_Additive, boneToModify {boneName=wing_R}, rotation=<rot 10 15 -15>})) {layerGroup=None}
   "Sequence Player (additive) 'MistRaven_Idle_PACK'"() {sequence=/BRCosmetics/Accessories/FORT_Backpacks/Backpack_F_MED_MistRaven/Animation/MistRaven_Idle_PACK}
   AnimGraphNode_StateMachine_2 = wingPosition() {is=title AnimGraphNode_StateMachine}
  ]}
 IdleBreak {
  entry=Idle Pose
  path=AnimGraph.AnimGraphNode_StateMachine_0.IdleBreak
  ids=id
  nodes [
   AnimStateNode_0 = Idle Pose() {is=title AnimStateNode}
   AnimStateNode_1 = Idle Break() {is=title AnimStateNode}
   AnimStateTransitionNode_0 = AnimStateTransitionNode()
   AnimStateTransitionNode_1 = AnimStateTransitionNode() {bAutomaticRuleBasedOnSequencePlayerInState=true}
  ]
  links [
   {
    dst=Idle Pose
    rule=AnimGraph.AnimGraphNode_StateMachine_0.IdleBreak.AnimStateTransitionNode_1.Transition
    src=Idle Break
    via=AnimStateTransitionNode_1
   }
   {
    dst=Idle Break
    rule=AnimGraph.AnimGraphNode_StateMachine_0.IdleBreak.AnimStateTransitionNode_0.Transition
    src=Idle Pose
    via=AnimStateTransitionNode_0
   }
  ]
 }
 Idle Pose {
  path=AnimGraph.AnimGraphNode_StateMachine_0.IdleBreak.AnimStateNode_0.Idle Pose
  nodes [Output Animation Pose(Result="Sequence Player (additive) 'TrackWaver_Yes_Idle_PACK'"() {sequence=/BRCosmetics/Accessories/FORT_Backpacks/Backpack_F_MED_TrackWaver_Yes/Animation/TrackWaver_Yes_Idle_PACK})]
 }
 Idle Break {
  path=AnimGraph.AnimGraphNode_StateMachine_0.IdleBreak.AnimStateNode_1.Idle Break
  nodes [
   n1 = "Sequence Player (additive) 'TrackWaver_Yes_IdleBreak_PACK'"() {bLoopAnimation=false, sequence=/BRCosmetics/Accessories/FORT_Backpacks/Backpack_F_MED_TrackWaver_Yes/Animation/TrackWaver_Yes_IdleBreak_PACK}
   Output Animation Pose(Result=@n1)
  ]
 }
 Transition {
  path=AnimGraph.AnimGraphNode_StateMachine_0.IdleBreak.AnimStateTransitionNode_0.Transition
  nodes [Result(bCanEnterTransition=Get Play_IdleBreak() -> (Play_IdleBreak:bool))]
 }
 Transition {path=AnimGraph.AnimGraphNode_StateMachine_0.IdleBreak.AnimStateTransitionNode_1.Transition, nodes [Result()]}
 wingPosition_1 {
  entry=Base
  path=AnimGraph.AnimGraphNode_StateMachine_1.wingPosition_1
  ids=id
  nodes [
   AnimStateNode_0 = Base() {is=title AnimStateNode}
   AnimStateNode_1 = Jumping() {is=title AnimStateNode}
   AnimStateNode_2 = Falling() {is=title AnimStateNode}
   AnimStateNode_3 = Targeting() {is=title AnimStateNode}
   AnimStateNode_5 = SkydiveMode() {is=title AnimStateNode}
   AnimStateNode_9 = Crouching() {is=title AnimStateNode}
   AnimStateTransitionNode_0 = AnimStateTransitionNode() {blendMode=Custom, customBlendCurve=/Game/Accessories/FORT_Backpacks/M_MED_Demon_Pack/Animation/JumpTransitionBlendCurve, priorityOrder=2}
   AnimStateTransitionNode_1 = AnimStateTransitionNode() {
    blendMode=Custom
    crossfadeDuration=0.5
    customBlendCurve=/Game/Accessories/FORT_Backpacks/M_MED_Demon_Pack/Animation/JumpTransitionBlendOutCurve
   }
   AnimStateTransitionNode_11 = AnimStateTransitionNode()
   AnimStateTransitionNode_12 = AnimStateTransitionNode()
   AnimStateTransitionNode_13 = AnimStateTransitionNode()
   AnimStateTransitionNode_15 = AnimStateTransitionNode()
   AnimStateTransitionNode_2 = AnimStateTransitionNode()
   AnimStateTransitionNode_3 = AnimStateTransitionNode() {blendMode=Custom, customBlendCurve=&JumpTransitionBlendOutCurve}
   AnimStateTransitionNode_4 = AnimStateTransitionNode() {blendMode=Custom, customBlendCurve=&JumpTransitionBlendCurve}
   AnimStateTransitionNode_5 = AnimStateTransitionNode()
   AnimStateTransitionNode_6 = AnimStateTransitionNode() {crossfadeDuration=0.25}
   AnimStateTransitionNode_7 = AnimStateTransitionNode() {crossfadeDuration=0.25}
  ]
  links [
   {
    dst=Crouching
    rule=AnimGraph.AnimGraphNode_StateMachine_1.wingPosition_1.AnimStateTransitionNode_12.Transition
    src=Base
    via=AnimStateTransitionNode_12
   }
   {
    dst=Falling
    rule=AnimGraph.AnimGraphNode_StateMachine_1.wingPosition_1.AnimStateTransitionNode_4.Transition
    src=Base
    via=AnimStateTransitionNode_4
   }
   {
    dst=Jumping
    rule=AnimGraph.AnimGraphNode_StateMachine_1.wingPosition_1.AnimStateTransitionNode_0.Transition
    src=Base
    via=AnimStateTransitionNode_0
   }
   {
    dst=SkydiveMode
    rule=AnimGraph.AnimGraphNode_StateMachine_1.wingPosition_1.AnimStateTransitionNode_15.Transition
    src=Base
    via=AnimStateTransitionNode_15
   }
   {
    dst=Targeting
    rule=AnimGraph.AnimGraphNode_StateMachine_1.wingPosition_1.AnimStateTransitionNode_6.Transition
    src=Base
    via=AnimStateTransitionNode_6
   }
   {
    dst=Base
    rule=AnimGraph.AnimGraphNode_StateMachine_1.wingPosition_1.AnimStateTransitionNode_13.Transition
    src=Crouching
    via=AnimStateTransitionNode_13
   }
   {
    dst=Base
    rule=AnimGraph.AnimGraphNode_StateMachine_1.wingPosition_1.AnimStateTransitionNode_3.Transition
    src=Falling
    via=AnimStateTransitionNode_3
   }
   {
    dst=Base
    rule=AnimGraph.AnimGraphNode_StateMachine_1.wingPosition_1.AnimStateTransitionNode_1.Transition
    src=Jumping
    via=AnimStateTransitionNode_1
   }
   {
    dst=Falling
    rule=AnimGraph.AnimGraphNode_StateMachine_1.wingPosition_1.AnimStateTransitionNode_2.Transition
    src=Jumping
    via=AnimStateTransitionNode_2
   }
   {
    dst=SkydiveMode
    rule=AnimGraph.AnimGraphNode_StateMachine_1.wingPosition_1.AnimStateTransitionNode_5.Transition
    src=Jumping
    via=AnimStateTransitionNode_5
   }
   {
    dst=Base
    rule=AnimGraph.AnimGraphNode_StateMachine_1.wingPosition_1.AnimStateTransitionNode_11.Transition
    src=SkydiveMode
    via=AnimStateTransitionNode_11
   }
   {
    dst=Base
    rule=AnimGraph.AnimGraphNode_StateMachine_1.wingPosition_1.AnimStateTransitionNode_7.Transition
    src=Targeting
    via=AnimStateTransitionNode_7
   }
  ]
 }
 Base {
  path=AnimGraph.AnimGraphNode_StateMachine_1.wingPosition_1.AnimStateNode_0.Base
  nodes [
   n1 = Get PlayerVector_FwdBwd() -> (PlayerVector_FwdBwd:double)
   n2 = Get PlayerVector_LR() -> (PlayerVector_LR:double)
   n3 = TrackWaver_Yes_Wings_Movement_Poses_PACK_PoseAsset : Base!() -> (Pose:Pose Link Structure) {
    is=title AnimGraphNode_PoseByName
    poseAsset=/BRCosmetics/Accessories/FORT_Backpacks/Backpack_F_MED_TrackWaver_Yes/Animation/TrackWaver_Yes_Wings_Movement_Poses_PACK_PoseAsset
    poseName=Base
    poseWeight=0.8
   }
   Output Animation Pose(Result=Apply Additive(
     Base=@n3
     Additive=Blendspace Player 'BS_TrackWaver_Yes_Leans_Ground'(X=@n2, Y=@n1) {blendSpace=/BRCosmetics/Accessories/FORT_Backpacks/Backpack_F_MED_TrackWaver_Yes/Animation/BS_TrackWaver_Yes_Leans_Ground}
    ) {alphaBoolBlend {blendOption=Linear}})
  ]
 }
 Jumping {
  path=AnimGraph.AnimGraphNode_StateMachine_1.wingPosition_1.AnimStateNode_1.Jumping
  nodes [
   n1 = Get PlayerVector_FwdBwd()
   n2 = Get PlayerVector_LR()
   n3 = TrackWaver_Yes_Wings_Movement_Poses_PACK_PoseAsset : Jump!(PoseWeight:float=0.6) -> (Pose:Pose Link Structure) {
    is=title AnimGraphNode_PoseByName
    poseAsset=&TrackWaver_Yes_Wings_Movement_Poses_PACK_PoseAsset
    poseName=Jump
   }
   Output Animation Pose(Result=Apply Additive(
     Base=@n3
     Additive=Blendspace Player 'BS_TrackWaver_Yes_Leans_Ground'(X=@n2, Y=@n1) {blendSpace=&BS_TrackWaver_Yes_Leans_Ground}
    ) {alphaBoolBlend {blendOption=Linear}})
  ]
 }
 Falling {
  path=AnimGraph.AnimGraphNode_StateMachine_1.wingPosition_1.AnimStateNode_2.Falling
  nodes [
   n1 = Get PlayerVector_FwdBwd()
   n2 = Get PlayerVector_LR()
   n3 = TrackWaver_Yes_Wings_Movement_Poses_PACK_PoseAsset : Jump(PoseWeight=0.6) {
    poseAsset=&TrackWaver_Yes_Wings_Movement_Poses_PACK_PoseAsset
    poseName=Jump
   }
   Output Animation Pose(Result=Apply Additive(
     Base=@n3
     Additive=Blendspace Player 'BS_TrackWaver_Yes_Leans_Ground'(X=@n2, Y=@n1) {blendSpace=&BS_TrackWaver_Yes_Leans_Ground}
    ) {alphaBoolBlend {blendOption=Linear}})
  ]
 }
 Targeting {
  path=AnimGraph.AnimGraphNode_StateMachine_1.wingPosition_1.AnimStateNode_3.Targeting
  nodes [
   n1 = TrackWaver_Yes_Wings_Movement_Poses_PACK_PoseAsset : Targeting!(PoseWeight:float=0.8) -> (Pose:Pose Link Structure) {
    is=title AnimGraphNode_PoseByName
    poseAsset=&TrackWaver_Yes_Wings_Movement_Poses_PACK_PoseAsset
    poseName=Targeting
   }
   Output Animation Pose(Result=@n1)
  ]
 }
 SkydiveMode {
  path=AnimGraph.AnimGraphNode_StateMachine_1.wingPosition_1.AnimStateNode_5.SkydiveMode
  nodes [
   AnimGraphNode_StateMachine_0 = Skydiving State() {is=title AnimGraphNode_StateMachine}
   Output Animation Pose(Result=@AnimGraphNode_StateMachine_0)
  ]
 }
 Skydiving State {
  entry=Parachuting
  path=AnimGraph.AnimGraphNode_StateMachine_1.wingPosition_1.AnimStateNode_5.SkydiveMode.AnimGraphNode_StateMachine_0.Skydiving State
  ids=id
  nodes [
   AnimStateNode_1 = Diving() {is=title AnimStateNode}
   AnimStateNode_2 = Floating() {is=title AnimStateNode}
   AnimStateNode_4 = Parachuting() {is=title AnimStateNode}
   AnimStateTransitionNode_1 = AnimStateTransitionNode() {crossfadeDuration=0.5}
   AnimStateTransitionNode_10 = AnimStateTransitionNode()
   AnimStateTransitionNode_12 = AnimStateTransitionNode()
   AnimStateTransitionNode_13 = AnimStateTransitionNode()
   AnimStateTransitionNode_3 = AnimStateTransitionNode()
   AnimStateTransitionNode_5 = AnimStateTransitionNode() {crossfadeDuration=0.5}
  ]
  links [
   {
    dst=Floating
    rule=AnimGraph.AnimGraphNode_StateMachine_1.wingPosition_1.AnimStateNode_5.SkydiveMode.AnimGraphNode_StateMachine_0.Skydiving State.AnimStateTransitionNode_1.AnimationTransitionGraph_0
    src=Diving
    via=AnimStateTransitionNode_1
   }
   {
    dst=Parachuting
    rule=AnimGraph.AnimGraphNode_StateMachine_1.wingPosition_1.AnimStateNode_5.SkydiveMode.AnimGraphNode_StateMachine_0.Skydiving State.AnimStateTransitionNode_10.AnimationTransitionGraph_1
    src=Diving
    via=AnimStateTransitionNode_10
   }
   {
    dst=Diving
    rule=AnimGraph.AnimGraphNode_StateMachine_1.wingPosition_1.AnimStateNode_5.SkydiveMode.AnimGraphNode_StateMachine_0.Skydiving State.AnimStateTransitionNode_5.AnimationTransitionGraph_0
    src=Floating
    via=AnimStateTransitionNode_5
   }
   {
    dst=Parachuting
    rule=AnimGraph.AnimGraphNode_StateMachine_1.wingPosition_1.AnimStateNode_5.SkydiveMode.AnimGraphNode_StateMachine_0.Skydiving State.AnimStateTransitionNode_3.AnimationTransitionGraph_1
    src=Floating
    via=AnimStateTransitionNode_3
   }
   {
    dst=Diving
    rule=AnimGraph.AnimGraphNode_StateMachine_1.wingPosition_1.AnimStateNode_5.SkydiveMode.AnimGraphNode_StateMachine_0.Skydiving State.AnimStateTransitionNode_13.AnimationTransitionGraph_1
    src=Parachuting
    via=AnimStateTransitionNode_13
   }
   {
    dst=Floating
    rule=AnimGraph.AnimGraphNode_StateMachine_1.wingPosition_1.AnimStateNode_5.SkydiveMode.AnimGraphNode_StateMachine_0.Skydiving State.AnimStateTransitionNode_12.AnimationTransitionGraph_1
    src=Parachuting
    via=AnimStateTransitionNode_12
   }
  ]
 }
 Diving {
  path=AnimGraph.AnimGraphNode_StateMachine_1.wingPosition_1.AnimStateNode_5.SkydiveMode.AnimGraphNode_StateMachine_0.Skydiving State.AnimStateNode_1.Diving
  nodes [
   n1 = Get PlayerVector_FwdBwd()
   n2 = Get PlayerVector_LR()
   n3 = TrackWaver_Yes_Wings_Movement_Poses_PACK_PoseAsset : Dive!(PoseWeight:float=0.4) -> (Pose:Pose Link Structure) {
    is=title AnimGraphNode_PoseByName
    poseAsset=&TrackWaver_Yes_Wings_Movement_Poses_PACK_PoseAsset
    poseName=Dive
   }
   Output Animation Pose(Result=Apply Additive(
     Base=@n3
     Additive=Blendspace Player 'BS_TrackWaver_Yes_Wings_Leans_Air'(X=@n2, Y=@n1) {blendSpace=/BRCosmetics/Accessories/FORT_Backpacks/Backpack_F_MED_TrackWaver_Yes/Animation/BS_TrackWaver_Yes_Wings_Leans_Air}
     Alpha=0.9
    ) {alphaBoolBlend {blendOption=Linear}})
  ]
 }
 Floating {
  path=AnimGraph.AnimGraphNode_StateMachine_1.wingPosition_1.AnimStateNode_5.SkydiveMode.AnimGraphNode_StateMachine_0.Skydiving State.AnimStateNode_2.Floating
  nodes [
   n1 = Get PlayerVector_FwdBwd()
   n2 = Get PlayerVector_LR()
   n3 = TrackWaver_Yes_Wings_Movement_Poses_PACK_PoseAsset : Float!(PoseWeight:float=0.8) -> (Pose:Pose Link Structure) {
    is=title AnimGraphNode_PoseByName
    poseAsset=&TrackWaver_Yes_Wings_Movement_Poses_PACK_PoseAsset
    poseName=Float
   }
   Output Animation Pose(Result=Apply Additive(
     Base=@n3
     Additive=Blendspace Player 'BS_TrackWaver_Yes_Wings_Leans_Air'(X=@n2, Y=@n1) {blendSpace=&BS_TrackWaver_Yes_Wings_Leans_Air}
     Alpha=0.9
    ) {alphaBoolBlend {blendOption=Linear}})
  ]
 }
 Parachuting {
  path=AnimGraph.AnimGraphNode_StateMachine_1.wingPosition_1.AnimStateNode_5.SkydiveMode.AnimGraphNode_StateMachine_0.Skydiving State.AnimStateNode_4.Parachuting
  nodes [
   n1 = Get PlayerVector_FwdBwd()
   n2 = Get PlayerVector_LR()
   n3 = TrackWaver_Yes_Wings_Movement_Poses_PACK_PoseAsset : Float(PoseWeight=0.8) {
    poseAsset=&TrackWaver_Yes_Wings_Movement_Poses_PACK_PoseAsset
    poseName=Float
   }
   Output Animation Pose(Result=Apply Additive(
     Base=@n3
     Additive=Blendspace Player 'BS_TrackWaver_Yes_Wings_Leans_Air'(X=@n2, Y=@n1) {blendSpace=&BS_TrackWaver_Yes_Wings_Leans_Air}
    ) {alphaBoolBlend {blendOption=Linear}})
  ]
 }
 AnimationTransitionGraph_0 {
  path=AnimGraph.AnimGraphNode_StateMachine_1.wingPosition_1.AnimStateNode_5.SkydiveMode.AnimGraphNode_StateMachine_0.Skydiving State.AnimStateTransitionNode_1.AnimationTransitionGraph_0
  nodes [Result(bCanEnterTransition=NOT Boolean(A=Get bIsSkydiveDiveMode()))]
 }
 AnimationTransitionGraph_1 {
  path=AnimGraph.AnimGraphNode_StateMachine_1.wingPosition_1.AnimStateNode_5.SkydiveMode.AnimGraphNode_StateMachine_0.Skydiving State.AnimStateTransitionNode_10.AnimationTransitionGraph_1
  nodes [Result(bCanEnterTransition=Get bIsParachuteOpened())]
 }
 AnimationTransitionGraph_1 {
  path=AnimGraph.AnimGraphNode_StateMachine_1.wingPosition_1.AnimStateNode_5.SkydiveMode.AnimGraphNode_StateMachine_0.Skydiving State.AnimStateTransitionNode_12.AnimationTransitionGraph_1
  nodes [Result(bCanEnterTransition=NOT Boolean(A=Get bIsParachuteOpened()))]
 }
 AnimationTransitionGraph_1 {
  path=AnimGraph.AnimGraphNode_StateMachine_1.wingPosition_1.AnimStateNode_5.SkydiveMode.AnimGraphNode_StateMachine_0.Skydiving State.AnimStateTransitionNode_13.AnimationTransitionGraph_1
  nodes [Result(bCanEnterTransition=NOT Boolean(A=Get bIsParachuteOpened()))]
 }
 AnimationTransitionGraph_1 {
  path=AnimGraph.AnimGraphNode_StateMachine_1.wingPosition_1.AnimStateNode_5.SkydiveMode.AnimGraphNode_StateMachine_0.Skydiving State.AnimStateTransitionNode_3.AnimationTransitionGraph_1
  nodes [Result(bCanEnterTransition=Get bIsParachuteOpened())]
 }
 AnimationTransitionGraph_0 {
  path=AnimGraph.AnimGraphNode_StateMachine_1.wingPosition_1.AnimStateNode_5.SkydiveMode.AnimGraphNode_StateMachine_0.Skydiving State.AnimStateTransitionNode_5.AnimationTransitionGraph_0
  nodes [Result(bCanEnterTransition=Get bIsSkydiveDiveMode())]
 }
 Crouching {
  path=AnimGraph.AnimGraphNode_StateMachine_1.wingPosition_1.AnimStateNode_9.Crouching
  nodes [
   n1 = TrackWaver_Yes_Wings_Movement_Poses_PACK_PoseAsset : Run!(PoseWeight:float=1) -> (Pose:Pose Link Structure) {
    is=title AnimGraphNode_PoseByName
    poseAsset=&TrackWaver_Yes_Wings_Movement_Poses_PACK_PoseAsset
    poseName=Run
   }
   n2 = TrackWaver_Yes_Wings_Movement_Poses_PACK_PoseAsset : Crouch!(PoseWeight:float=1) -> (Pose:Pose Link Structure) {
    is=title AnimGraphNode_PoseByName
    poseAsset=&TrackWaver_Yes_Wings_Movement_Poses_PACK_PoseAsset
    poseName=Crouch
   }
   Output Animation Pose(Result=Layered blend per bone(BasePose=@n2, BlendPoses_0=@n1, BlendWeights_0=0.6) {layerSetup [{branchFilters [{blendDepth=3, boneName=WingRoot}]}]})
  ]
 }
 Transition {
  path=AnimGraph.AnimGraphNode_StateMachine_1.wingPosition_1.AnimStateTransitionNode_0.Transition
  nodes [Result(bCanEnterTransition=Get bIsJumping())]
 }
 Transition {
  path=AnimGraph.AnimGraphNode_StateMachine_1.wingPosition_1.AnimStateTransitionNode_1.Transition
  nodes [
   Result(bCanEnterTransition=NOT Boolean(A=Get bIsJumping()))
   Get bShouldPredictLanding() -> (bShouldPredictLanding:bool)
   Get bLandingPredicted() -> (bLandingPredicted:bool)
   Get bDebugLandPrediction() -> (bDebugLandPrediction:bool)
   Get bIsFalling()
  ]
 }
 Transition {
  path=AnimGraph.AnimGraphNode_StateMachine_1.wingPosition_1.AnimStateTransitionNode_11.Transition
  nodes [Result(bCanEnterTransition=NOT Boolean(A=Get bIsSkyDiving()))]
 }
 Transition {
  path=AnimGraph.AnimGraphNode_StateMachine_1.wingPosition_1.AnimStateTransitionNode_12.Transition
  nodes [Result(bCanEnterTransition=Get bIsCrouching())]
 }
 Transition {
  path=AnimGraph.AnimGraphNode_StateMachine_1.wingPosition_1.AnimStateTransitionNode_13.Transition
  nodes [Result(bCanEnterTransition=NOT Boolean(A=Get bIsCrouching()))]
 }
 Transition {
  path=AnimGraph.AnimGraphNode_StateMachine_1.wingPosition_1.AnimStateTransitionNode_15.Transition
  nodes [Result(bCanEnterTransition=Get bIsSkyDiving())]
 }
 Transition {
  path=AnimGraph.AnimGraphNode_StateMachine_1.wingPosition_1.AnimStateTransitionNode_2.Transition
  nodes [Result(bCanEnterTransition=Get bIsFalling())]
 }
 Transition {
  path=AnimGraph.AnimGraphNode_StateMachine_1.wingPosition_1.AnimStateTransitionNode_3.Transition
  nodes [Result(bCanEnterTransition=NOT Boolean(A=Get bIsFalling()))]
 }
 Transition {
  path=AnimGraph.AnimGraphNode_StateMachine_1.wingPosition_1.AnimStateTransitionNode_4.Transition
  nodes [Result(bCanEnterTransition=Get bIsFalling())]
 }
 Transition {
  path=AnimGraph.AnimGraphNode_StateMachine_1.wingPosition_1.AnimStateTransitionNode_5.Transition
  nodes [Result(bCanEnterTransition=Get bIsSkyDiving())]
 }
 Transition {
  path=AnimGraph.AnimGraphNode_StateMachine_1.wingPosition_1.AnimStateTransitionNode_6.Transition
  nodes [
   Result(bCanEnterTransition=Get bIsTargeting())
   Get is Targeting() -> (is Targeting:bool)
  ]
 }
 Transition {
  path=AnimGraph.AnimGraphNode_StateMachine_1.wingPosition_1.AnimStateTransitionNode_7.Transition
  nodes [Result(bCanEnterTransition=NOT Boolean(A=Get bIsTargeting()))]
 }
 wingPosition {
  entry=Base
  path=AnimGraph.AnimGraphNode_StateMachine_2.wingPosition
  ids=id
  nodes [
   AnimStateNode_0 = Base()
   AnimStateNode_1 = Jumping()
   AnimStateNode_2 = Falling()
   AnimStateNode_3 = Targeting()
   AnimStateNode_5 = SkydiveMode()
   AnimStateNode_9 = Crouching()
   AnimStateTransitionNode_0 = AnimStateTransitionNode() {blendMode=Custom, customBlendCurve=&JumpTransitionBlendCurve, priorityOrder=2}
   AnimStateTransitionNode_1 = AnimStateTransitionNode() {
    blendMode=Custom
    crossfadeDuration=0.5
    customBlendCurve=&JumpTransitionBlendOutCurve
   }
   AnimStateTransitionNode_11 = AnimStateTransitionNode()
   AnimStateTransitionNode_12 = AnimStateTransitionNode()
   AnimStateTransitionNode_13 = AnimStateTransitionNode()
   AnimStateTransitionNode_15 = AnimStateTransitionNode()
   AnimStateTransitionNode_2 = AnimStateTransitionNode()
   AnimStateTransitionNode_3 = AnimStateTransitionNode() {blendMode=Custom, customBlendCurve=&JumpTransitionBlendOutCurve}
   AnimStateTransitionNode_4 = AnimStateTransitionNode() {blendMode=Custom, customBlendCurve=&JumpTransitionBlendCurve}
   AnimStateTransitionNode_5 = AnimStateTransitionNode()
   AnimStateTransitionNode_6 = AnimStateTransitionNode() {crossfadeDuration=0.25}
   AnimStateTransitionNode_7 = AnimStateTransitionNode() {crossfadeDuration=0.25}
  ]
  links [
   {
    dst=Crouching
    rule=AnimGraph.AnimGraphNode_StateMachine_2.wingPosition.AnimStateTransitionNode_12.Transition
    src=Base
    via=AnimStateTransitionNode_12
   }
   {
    dst=Falling
    rule=AnimGraph.AnimGraphNode_StateMachine_2.wingPosition.AnimStateTransitionNode_4.Transition
    src=Base
    via=AnimStateTransitionNode_4
   }
   {
    dst=Jumping
    rule=AnimGraph.AnimGraphNode_StateMachine_2.wingPosition.AnimStateTransitionNode_0.Transition
    src=Base
    via=AnimStateTransitionNode_0
   }
   {
    dst=SkydiveMode
    rule=AnimGraph.AnimGraphNode_StateMachine_2.wingPosition.AnimStateTransitionNode_15.Transition
    src=Base
    via=AnimStateTransitionNode_15
   }
   {
    dst=Targeting
    rule=AnimGraph.AnimGraphNode_StateMachine_2.wingPosition.AnimStateTransitionNode_6.Transition
    src=Base
    via=AnimStateTransitionNode_6
   }
   {
    dst=Base
    rule=AnimGraph.AnimGraphNode_StateMachine_2.wingPosition.AnimStateTransitionNode_13.Transition
    src=Crouching
    via=AnimStateTransitionNode_13
   }
   {
    dst=Base
    rule=AnimGraph.AnimGraphNode_StateMachine_2.wingPosition.AnimStateTransitionNode_3.Transition
    src=Falling
    via=AnimStateTransitionNode_3
   }
   {
    dst=Base
    rule=AnimGraph.AnimGraphNode_StateMachine_2.wingPosition.AnimStateTransitionNode_1.Transition
    src=Jumping
    via=AnimStateTransitionNode_1
   }
   {
    dst=Falling
    rule=AnimGraph.AnimGraphNode_StateMachine_2.wingPosition.AnimStateTransitionNode_2.Transition
    src=Jumping
    via=AnimStateTransitionNode_2
   }
   {
    dst=SkydiveMode
    rule=AnimGraph.AnimGraphNode_StateMachine_2.wingPosition.AnimStateTransitionNode_5.Transition
    src=Jumping
    via=AnimStateTransitionNode_5
   }
   {
    dst=Base
    rule=AnimGraph.AnimGraphNode_StateMachine_2.wingPosition.AnimStateTransitionNode_11.Transition
    src=SkydiveMode
    via=AnimStateTransitionNode_11
   }
   {
    dst=Base
    rule=AnimGraph.AnimGraphNode_StateMachine_2.wingPosition.AnimStateTransitionNode_7.Transition
    src=Targeting
    via=AnimStateTransitionNode_7
   }
  ]
 }
 Base {
  path=AnimGraph.AnimGraphNode_StateMachine_2.wingPosition.AnimStateNode_0.Base
  nodes [
   n1 = Get PlayerVector_FwdBwd()
   n2 = Get PlayerVector_LR()
   n3 = MistRaven_Wings_Movement_Poses_PoseAsset : Base!() -> (Pose:Pose Link Structure) {
    is=title AnimGraphNode_PoseByName
    poseAsset=/BRCosmetics/Accessories/FORT_Backpacks/Backpack_F_MED_MistRaven/Animation/MistRaven_Wings_Movement_Poses_PoseAsset
    poseName=Base
    poseWeight=0.8
   }
   Output Animation Pose(Result=Apply Additive(
     Base=@n3
     Additive=Blendspace Player 'BS_MistRaven_Leans_Ground'(X=@n2, Y=@n1) {blendSpace=/BRCosmetics/Accessories/FORT_Backpacks/Backpack_F_MED_MistRaven/Animation/BS_MistRaven_Leans_Ground}
    ) {alphaBoolBlend {blendOption=Linear}})
  ]
 }
 Jumping {
  path=AnimGraph.AnimGraphNode_StateMachine_2.wingPosition.AnimStateNode_1.Jumping
  nodes [
   n1 = Get PlayerVector_FwdBwd()
   n2 = Get PlayerVector_LR()
   n3 = MistRaven_Wings_Movement_Poses_PoseAsset : Jump!(PoseWeight:float=0.8) -> (Pose:Pose Link Structure) {
    is=title AnimGraphNode_PoseByName
    poseAsset=&MistRaven_Wings_Movement_Poses_PoseAsset
    poseName=Jump
   }
   Output Animation Pose(Result=Apply Additive(
     Base=@n3
     Additive=Blendspace Player 'BS_MistRaven_Leans_Ground'(X=@n2, Y=@n1) {blendSpace=&BS_MistRaven_Leans_Ground}
    ) {alphaBoolBlend {blendOption=Linear}})
  ]
 }
 Falling {
  path=AnimGraph.AnimGraphNode_StateMachine_2.wingPosition.AnimStateNode_2.Falling
  nodes [
   n1 = Get PlayerVector_FwdBwd()
   n2 = Get PlayerVector_LR()
   n3 = MistRaven_Wings_Movement_Poses_PoseAsset : Jump(PoseWeight=0.8) {poseAsset=&MistRaven_Wings_Movement_Poses_PoseAsset, poseName=Jump}
   Output Animation Pose(Result=Apply Additive(
     Base=@n3
     Additive=Blendspace Player 'BS_MistRaven_Leans_Ground'(X=@n2, Y=@n1) {blendSpace=&BS_MistRaven_Leans_Ground}
    ) {alphaBoolBlend {blendOption=Linear}})
  ]
 }
 Targeting {
  path=AnimGraph.AnimGraphNode_StateMachine_2.wingPosition.AnimStateNode_3.Targeting
  nodes [
   n1 = MistRaven_Wings_Movement_Poses_PoseAsset : Targeting!(PoseWeight:float=0.8) -> (Pose:Pose Link Structure) {
    is=title AnimGraphNode_PoseByName
    poseAsset=&MistRaven_Wings_Movement_Poses_PoseAsset
    poseName=Targeting
   }
   Output Animation Pose(Result=@n1)
  ]
 }
 SkydiveMode {
  path=AnimGraph.AnimGraphNode_StateMachine_2.wingPosition.AnimStateNode_5.SkydiveMode
  nodes [
   AnimGraphNode_StateMachine_0 = Skydiving State()
   Output Animation Pose(Result=@AnimGraphNode_StateMachine_0)
  ]
 }
 Skydiving State {
  entry=Parachuting
  path=AnimGraph.AnimGraphNode_StateMachine_2.wingPosition.AnimStateNode_5.SkydiveMode.AnimGraphNode_StateMachine_0.Skydiving State
  ids=id
  nodes [
   AnimStateNode_1 = Diving()
   AnimStateNode_2 = Floating()
   AnimStateNode_4 = Parachuting()
   AnimStateTransitionNode_1 = AnimStateTransitionNode() {crossfadeDuration=0.5}
   AnimStateTransitionNode_10 = AnimStateTransitionNode()
   AnimStateTransitionNode_12 = AnimStateTransitionNode()
   AnimStateTransitionNode_13 = AnimStateTransitionNode()
   AnimStateTransitionNode_3 = AnimStateTransitionNode()
   AnimStateTransitionNode_5 = AnimStateTransitionNode() {crossfadeDuration=0.5}
  ]
  links [
   {
    dst=Floating
    rule=AnimGraph.AnimGraphNode_StateMachine_2.wingPosition.AnimStateNode_5.SkydiveMode.AnimGraphNode_StateMachine_0.Skydiving State.AnimStateTransitionNode_1.AnimationTransitionGraph_0
    src=Diving
    via=AnimStateTransitionNode_1
   }
   {
    dst=Parachuting
    rule=AnimGraph.AnimGraphNode_StateMachine_2.wingPosition.AnimStateNode_5.SkydiveMode.AnimGraphNode_StateMachine_0.Skydiving State.AnimStateTransitionNode_10.AnimationTransitionGraph_1
    src=Diving
    via=AnimStateTransitionNode_10
   }
   {
    dst=Diving
    rule=AnimGraph.AnimGraphNode_StateMachine_2.wingPosition.AnimStateNode_5.SkydiveMode.AnimGraphNode_StateMachine_0.Skydiving State.AnimStateTransitionNode_5.AnimationTransitionGraph_0
    src=Floating
    via=AnimStateTransitionNode_5
   }
   {
    dst=Parachuting
    rule=AnimGraph.AnimGraphNode_StateMachine_2.wingPosition.AnimStateNode_5.SkydiveMode.AnimGraphNode_StateMachine_0.Skydiving State.AnimStateTransitionNode_3.AnimationTransitionGraph_1
    src=Floating
    via=AnimStateTransitionNode_3
   }
   {
    dst=Diving
    rule=AnimGraph.AnimGraphNode_StateMachine_2.wingPosition.AnimStateNode_5.SkydiveMode.AnimGraphNode_StateMachine_0.Skydiving State.AnimStateTransitionNode_13.AnimationTransitionGraph_1
    src=Parachuting
    via=AnimStateTransitionNode_13
   }
   {
    dst=Floating
    rule=AnimGraph.AnimGraphNode_StateMachine_2.wingPosition.AnimStateNode_5.SkydiveMode.AnimGraphNode_StateMachine_0.Skydiving State.AnimStateTransitionNode_12.AnimationTransitionGraph_1
    src=Parachuting
    via=AnimStateTransitionNode_12
   }
  ]
 }
 Diving {
  path=AnimGraph.AnimGraphNode_StateMachine_2.wingPosition.AnimStateNode_5.SkydiveMode.AnimGraphNode_StateMachine_0.Skydiving State.AnimStateNode_1.Diving
  nodes [
   n1 = Get PlayerVector_FwdBwd()
   n2 = Get PlayerVector_LR()
   n3 = MistRaven_Wings_Movement_Poses_PoseAsset : Dive!(PoseWeight:float=0.7) -> (Pose:Pose Link Structure) {
    is=title AnimGraphNode_PoseByName
    poseAsset=&MistRaven_Wings_Movement_Poses_PoseAsset
    poseName=Dive
   }
   Output Animation Pose(Result=Apply Additive(
     Base=@n3
     Additive=Blendspace Player 'BS_MistRaven_Wings_Leans_Air'(X=@n2, Y=@n1) {blendSpace=/BRCosmetics/Accessories/FORT_Backpacks/Backpack_F_MED_MistRaven/Animation/BS_MistRaven_Wings_Leans_Air}
     Alpha=0.9
    ) {alphaBoolBlend {blendOption=Linear}})
  ]
 }
 Floating {
  path=AnimGraph.AnimGraphNode_StateMachine_2.wingPosition.AnimStateNode_5.SkydiveMode.AnimGraphNode_StateMachine_0.Skydiving State.AnimStateNode_2.Floating
  nodes [
   n1 = Get PlayerVector_FwdBwd()
   n2 = Get PlayerVector_LR()
   n3 = MistRaven_Wings_Movement_Poses_PoseAsset : Float!(PoseWeight:float=0.7) -> (Pose:Pose Link Structure) {
    is=title AnimGraphNode_PoseByName
    poseAsset=&MistRaven_Wings_Movement_Poses_PoseAsset
    poseName=Float
   }
   Output Animation Pose(Result=Apply Additive(
     Base=@n3
     Additive=Blendspace Player 'BS_MistRaven_Wings_Leans_Air'(X=@n2, Y=@n1) {blendSpace=&BS_MistRaven_Wings_Leans_Air}
     Alpha=0.9
    ) {alphaBoolBlend {blendOption=Linear}})
  ]
 }
 Parachuting {
  path=AnimGraph.AnimGraphNode_StateMachine_2.wingPosition.AnimStateNode_5.SkydiveMode.AnimGraphNode_StateMachine_0.Skydiving State.AnimStateNode_4.Parachuting
  nodes [
   n1 = Get PlayerVector_FwdBwd()
   n2 = Get PlayerVector_LR()
   n3 = MistRaven_Wings_Movement_Poses_PoseAsset : Float(PoseWeight=0.8) {
    poseAsset=&MistRaven_Wings_Movement_Poses_PoseAsset
    poseName=Float
   }
   Output Animation Pose(Result=Apply Additive(
     Base=@n3
     Additive=Blendspace Player 'BS_MistRaven_Wings_Leans_Air'(X=@n2, Y=@n1) {blendSpace=&BS_MistRaven_Wings_Leans_Air}
    ) {alphaBoolBlend {blendOption=Linear}})
  ]
 }
 AnimationTransitionGraph_0 {
  path=AnimGraph.AnimGraphNode_StateMachine_2.wingPosition.AnimStateNode_5.SkydiveMode.AnimGraphNode_StateMachine_0.Skydiving State.AnimStateTransitionNode_1.AnimationTransitionGraph_0
  nodes [Result(bCanEnterTransition=NOT Boolean(A=Get bIsSkydiveDiveMode()))]
 }
 AnimationTransitionGraph_1 {
  path=AnimGraph.AnimGraphNode_StateMachine_2.wingPosition.AnimStateNode_5.SkydiveMode.AnimGraphNode_StateMachine_0.Skydiving State.AnimStateTransitionNode_10.AnimationTransitionGraph_1
  nodes [Result(bCanEnterTransition=Get bIsParachuteOpened())]
 }
 AnimationTransitionGraph_1 {
  path=AnimGraph.AnimGraphNode_StateMachine_2.wingPosition.AnimStateNode_5.SkydiveMode.AnimGraphNode_StateMachine_0.Skydiving State.AnimStateTransitionNode_12.AnimationTransitionGraph_1
  nodes [Result(bCanEnterTransition=NOT Boolean(A=Get bIsParachuteOpened()))]
 }
 AnimationTransitionGraph_1 {
  path=AnimGraph.AnimGraphNode_StateMachine_2.wingPosition.AnimStateNode_5.SkydiveMode.AnimGraphNode_StateMachine_0.Skydiving State.AnimStateTransitionNode_13.AnimationTransitionGraph_1
  nodes [Result(bCanEnterTransition=NOT Boolean(A=Get bIsParachuteOpened()))]
 }
 AnimationTransitionGraph_1 {
  path=AnimGraph.AnimGraphNode_StateMachine_2.wingPosition.AnimStateNode_5.SkydiveMode.AnimGraphNode_StateMachine_0.Skydiving State.AnimStateTransitionNode_3.AnimationTransitionGraph_1
  nodes [Result(bCanEnterTransition=Get bIsParachuteOpened())]
 }
 AnimationTransitionGraph_0 {
  path=AnimGraph.AnimGraphNode_StateMachine_2.wingPosition.AnimStateNode_5.SkydiveMode.AnimGraphNode_StateMachine_0.Skydiving State.AnimStateTransitionNode_5.AnimationTransitionGraph_0
  nodes [Result(bCanEnterTransition=Get bIsSkydiveDiveMode())]
 }
 Crouching {
  path=AnimGraph.AnimGraphNode_StateMachine_2.wingPosition.AnimStateNode_9.Crouching
  nodes [
   n1 = MistRaven_Wings_Movement_Poses_PoseAsset : Crouch!(PoseWeight:float=0.8) -> (Pose:Pose Link Structure) {
    is=title AnimGraphNode_PoseByName
    poseAsset=&MistRaven_Wings_Movement_Poses_PoseAsset
    poseName=Crouch
   }
   Output Animation Pose(Result=@n1)
  ]
 }
 Transition {
  path=AnimGraph.AnimGraphNode_StateMachine_2.wingPosition.AnimStateTransitionNode_0.Transition
  nodes [Result(bCanEnterTransition=Get bIsJumping())]
 }
 Transition {
  path=AnimGraph.AnimGraphNode_StateMachine_2.wingPosition.AnimStateTransitionNode_1.Transition
  nodes [
   Result(bCanEnterTransition=NOT Boolean(A=Get bIsJumping()))
   Get bShouldPredictLanding()
   Get bLandingPredicted()
   Get bDebugLandPrediction()
   Get bIsFalling()
  ]
 }
 Transition {
  path=AnimGraph.AnimGraphNode_StateMachine_2.wingPosition.AnimStateTransitionNode_11.Transition
  nodes [Result(bCanEnterTransition=NOT Boolean(A=Get bIsSkyDiving()))]
 }
 Transition {
  path=AnimGraph.AnimGraphNode_StateMachine_2.wingPosition.AnimStateTransitionNode_12.Transition
  nodes [Result(bCanEnterTransition=Get bIsCrouching())]
 }
 Transition {
  path=AnimGraph.AnimGraphNode_StateMachine_2.wingPosition.AnimStateTransitionNode_13.Transition
  nodes [Result(bCanEnterTransition=NOT Boolean(A=Get bIsCrouching()))]
 }
 Transition {
  path=AnimGraph.AnimGraphNode_StateMachine_2.wingPosition.AnimStateTransitionNode_15.Transition
  nodes [Result(bCanEnterTransition=Get bIsSkyDiving())]
 }
 Transition {
  path=AnimGraph.AnimGraphNode_StateMachine_2.wingPosition.AnimStateTransitionNode_2.Transition
  nodes [Result(bCanEnterTransition=Get bIsFalling())]
 }
 Transition {
  path=AnimGraph.AnimGraphNode_StateMachine_2.wingPosition.AnimStateTransitionNode_3.Transition
  nodes [Result(bCanEnterTransition=NOT Boolean(A=Get bIsFalling()))]
 }
 Transition {
  path=AnimGraph.AnimGraphNode_StateMachine_2.wingPosition.AnimStateTransitionNode_4.Transition
  nodes [Result(bCanEnterTransition=Get bIsFalling())]
 }
 Transition {
  path=AnimGraph.AnimGraphNode_StateMachine_2.wingPosition.AnimStateTransitionNode_5.Transition
  nodes [Result(bCanEnterTransition=Get bIsSkyDiving())]
 }
 Transition {
  path=AnimGraph.AnimGraphNode_StateMachine_2.wingPosition.AnimStateTransitionNode_6.Transition
  nodes [
   Result(bCanEnterTransition=Get bIsTargeting())
   Get is Targeting()
  ]
 }
 Transition {
  path=AnimGraph.AnimGraphNode_StateMachine_2.wingPosition.AnimStateTransitionNode_7.Transition
  nodes [Result(bCanEnterTransition=NOT Boolean(A=Get bIsTargeting()))]
 }
 EventGraph {nodes [
   Event BlueprintUpdateAnimation()
   > n1 = Sequence!(execute) -> (then_0, then_1, then_2:Exec, then_3:Exec)
   > n2 = PlayerMovement(AnimInstance=Self-Reference(), XDotProductCoefficient=1, YDotProductCoefficient=1, ZDotProductCoefficient=1)
   > n3 = Sequence!(execute) -> (then_0, then_1)
   > n4 = Branch(Condition=Get bIsAccelerating2D())
   > Set PlayerVector_LR(PlayerVector_LR:double=@n2) -> (Output_Get:double)
   Set PlayerVector_LR(execute=@n4.else, PlayerVector_LR=0)
   n5 = Branch(execute=@n3.then_1, Condition=Get bIsAccelerating2D())
   > Set PlayerVector_FwdBwd(PlayerVector_FwdBwd:double=@n2.YMovement) -> (Output_Get:double)
   Set PlayerVector_FwdBwd(execute=@n5.else, PlayerVector_FwdBwd=0)
   Delay(execute=@n1.then_1, Duration=RandomFloatInRange(Min=5, Max=20))
   > Set Play_IdleBreak(Play_IdleBreak:bool=true) -> (Output_Get:bool)
   > Delay(Duration=0.1)
   > Set Play_IdleBreak(Play_IdleBreak=false)
   Set Feather_AnimDynamics_Alpha(execute=@n1.then_2, Feather_AnimDynamics_Alpha:double=0.6) -> (Output_Get:double)
   > Set Wings_AnimDynamics_Alpha(Wings_AnimDynamics_Alpha:double=0.4) -> (Output_Get:double)
   n6 = Get PawnVelocity()
   n7 = AND Boolean(A=Get bIsCrouching(), B="float < float"(A=Vector Length(A=@n6), B=150))
   n8 = SelectFloat(A=0.8, B=1, bPickA=@n7)
   n9 = OR Boolean(A=AND Boolean(A=Get bIsTacticalSprinting(), B=Get bIsOnGround() -> (bIsOnGround:bool)), B=@n7)
   > Set wings fold(wings fold:float=FInterpTo(
     Current=Get wings fold()
     Target=SelectFloat(A=@n8, B=0, bPickA=@n9)
     DeltaTime=GetWorldDeltaSeconds!() -> (ReturnValue:double)
     InterpSpeed=8
    )) -> (Output_Get:float)
   n10 = Branch(execute=@n1.then_3, Condition=Get bIsInFrontEnd())
   > Set BackSpacing(BackSpacing:double=2) -> (Output_Get:double)
   Set BackSpacing(execute=@n10.else, BackSpacing=1)
   TryGetPawnOwner()
  ]}
]
variables [
 BackSpacing {type=number}
 Crouch {type=number}
 DisableDynamicsCurve {type=number}
 DivePose_alpha {type=number}
 Feather_AnimDynamics_Alpha {type=number, default=0.5}
 FlyPose_alpha {type=number}
 Gravity Override {type=Vector}
 Gravity Override Params {type=GravityOverrideParamsStruct, default {skydive_GoingUp {gravityOverride=<0 -700 600>, windOffset=<0 -600 400>}}}
 Jump {type=number}
 JumpPose_alpha {type=number}
 PackHidden {type=boolean}
 Play_IdleBreak {type=boolean}
 PlayerVector_FwdBwd {type=number}
 PlayerVector_LR {type=number}
 SprintJump {type=boolean}
 wings fold {type=number, default=1}
 Wings_AnimDynamics_Alpha {type=number, default=0.5}
]
